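# Physics asset for the StairGummy backpack cosmetic, defining collision bodies and joint limits for the pack and its dangling gloves, bottles and boxes.
# The pack is one unconstrained box; every other body is a capsule.
# Each bottle and box pairs with a tiny MSTR_ anchor body under tight swing and twist limits.
# Glove A hangs from its own anchor and leads through two looser rope links to glove B.
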PhysicsAsset /BRCosmetics/Accessories/FORT_Backpacks/Backpack_F_MED_StairGummy/Meshes/F_MED_StairGummy_Pack_Physics
bodies {
 MSTR_bottle_A [{shapeType=Capsule, length=0.505, radius=0.505, shapeName=MSTR_bottle_A_capsule, rotation=<rot 0 -90 0>}]
 MSTR_bottle_B [{shapeType=Capsule, length=0.505, radius=0.505, shapeName=MSTR_bottle_B_capsule, rotation=<rot 0 -90 0>}]
 MSTR_box_A [{shapeType=Capsule, length=0.505, radius=0.505, shapeName=MSTR_box_A_capsule, rotation=<rot 0 -90 0>}]
 MSTR_box_B [{shapeType=Capsule, length=0.505, radius=0.505, shapeName=MSTR_box_B_capsule, rotation=<rot 0 -90 0>}]
 MSTR_box_C [{shapeType=Capsule, length=0.505, radius=0.505, shapeName=MSTR_box_C_capsule, rotation=<rot 0 -90 0>}]
 MSTR_glove_A [{shapeType=Capsule, length=0.505, radius=0.505, shapeName=MSTR_glove_A_capsule, rotation=<rot 0 -90 0>}]
 bottle_A [{shapeType=Capsule, length=1, radius=1, shapeName=bottle_A_capsule}]
 bottle_B [{shapeType=Capsule, length=1, radius=1, shapeName=bottle_B_capsule}]
 box_A [{shapeType=Capsule, length=1, radius=1, shapeName=box_A_capsule, center=<0.0402 0 0>}]
 box_B [{shapeType=Capsule, length=1, radius=1, shapeName=box_B_capsule}]
 box_C [{shapeType=Capsule, length=1, radius=1, shapeName=box_C_capsule}]
 glove_A [{shapeType=Capsule, length=7, radius=4.5, shapeName=glove_A_capsule, center=<3.3137 0.6254 3.741>, rotation=<rot -80 -0.0006 -0.0002>}]
 glove_B [{
   shapeType=Capsule
   length=11.6011
   radius=4.4013
   shapeName=glove_B_capsule
   center=<4.9725 4.0145 0.5139>
   rotation=<rot 89.1578 -6.2649 0.0958>
  }]
 pack [{
   shapeType=Box
   extentX=23.4385
   extentY=16.0455
   extentZ=29.3827
   shapeName=pack_box
   center=<1.3032 -7.5915 1.0304>
   rotation=<rot 0.3912 8.9145 -3.4568>
  }]
 rope_1 [{shapeType=Capsule, length=3, radius=0.5, shapeName=rope_1_capsule, center=<3 0 0>, rotation=<rot -90 0.0003 0>}]
 rope_2 [{shapeType=Capsule, length=3, radius=0.5, shapeName=rope_2_capsule, center=<2 0 0>, rotation=<rot -90 -0.0007 0>}]
}
constraints [
 {bone1Name=rope_2, bone2Name=rope_1, swing1LimitDegrees=30, swing2LimitDegrees=30, twistLimitDegrees=2}
 {bone1Name=glove_B, bone2Name=rope_2, swing1LimitDegrees=30, swing2LimitDegrees=30, twistLimitDegrees=2}
 {bone1Name=glove_A, bone2Name=MSTR_glove_A, swing1LimitDegrees=5, swing2LimitDegrees=5, twistLimitDegrees=2}
 {bone1Name=box_C, bone2Name=MSTR_box_C, swing1LimitDegrees=2, swing2LimitDegrees=2, twistLimitDegrees=1}
 {bone1Name=box_B, bone2Name=MSTR_box_B, swing1LimitDegrees=2, swing2LimitDegrees=2, twistLimitDegrees=1}
 {bone1Name=box_A, bone2Name=MSTR_box_A, swing1LimitDegrees=2, swing2LimitDegrees=2, twistLimitDegrees=1}
 {bone1Name=bottle_B, bone2Name=MSTR_bottle_B, swing1LimitDegrees=2, swing2LimitDegrees=2, twistLimitDegrees=1}
 {bone1Name=bottle_A, bone2Name=MSTR_bottle_A, swing1LimitDegrees=2, swing2LimitDegrees=2, twistLimitDegrees=1}
 {bone1Name=rope_1, bone2Name=glove_A, swing1LimitDegrees=30, swing2LimitDegrees=30, twistLimitDegrees=1}
]
