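# Animation Blueprint for the StairGummy backpack cosmetic, adding physics motion and adjusting the pack's placement as the wearer's spine bends.
# A RigidBody node simulates the pack in base-bone space, with gravity overridden by a nativized, parameter-asset-driven state function evaluated each update.
# Two additive bone transforms follow: a small fixed offset, then a large shift and roll.
# The latter's alpha comes from the owner mesh's spine_05-to-spine_01 yaw delta from ref pose, range-mapped and clamped.
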AnimBlueprint /BRCosmetics/Accessories/FORT_Backpacks/Backpack_F_MED_StairGummy/Meshes/F_MED_StairGummy_Pack_AnimBP
target_skeleton=&/F_MED_StairGummy_Pack_Skeleton
graphs [
 AnimGraph {nodes [
   n1 = RigidBody!(ComponentPose=Local To Component(LocalPose=Input Pose()), Alpha, OverrideWorldGravity:Vector=Get Gravity Override(), ExternalForce) -> (Pose) {
    bOverrideWorldGravity=true
    overridePhysicsAsset=&/F_MED_StairGummy_Pack_Physics
    simulationSpace=BaseBoneSpace
    baseBoneRef {boneName=pack}
    componentAppliedLinearAccClamp=<0 0 0>
    simSpaceSettings {
     dampingAlpha=0
     maxAngularAcceleration=18
     maxAngularVelocity=9
     maxLinearAcceleration=1000
     maxLinearVelocity=400
     worldAlpha=1
     externalLinearDragV=<0.4 0.4 0.4>
    }
   }
   n2 = "Transform (Modify) Bone - Bone: pack"(ComponentPose=@n1, Translation="0,-0.5,-5.5", Rotation="0,0,-5") {
    rotationMode=BMM_Additive
    rotationSpace=BCS_BoneSpace
    translationMode=BMM_Additive
    translationSpace=BCS_BoneSpace
    boneToModify {boneName=pack}
    rotation {roll=-5}
    translation {y=-0.5, z=-5.5}
   }
   n3 = "Transform (Modify) Bone - Bone: pack"(ComponentPose=@n2, Alpha=Get transform_offset() -> (transform_offset:double), Translation="0,-35,5", Rotation="0,0,85") {
    rotationMode=BMM_Additive
    rotationSpace=BCS_BoneSpace
    translationMode=BMM_Additive
    translationSpace=BCS_BoneSpace
    boneToModify {boneName=pack}
    rotation {roll=85}
    translation {y=-35, z=5}
   }
   Output Pose(Result=Component To Local(ComponentPose=@n3)) {layerGroup=None}
  ]}
 EventGraph {nodes [
   Event BlueprintUpdateAnimation()
   > n1 = "Gravity Override State Function (Parameter Asset) (Nativized)"(
    CharacterPartAnimInstance=Self-Reference()
    GameTimeInSeconds=GetGameTimeInSeconds()
    Parameters=&/F_MED_StaitGummy_Params
   )
   > Set Gravity Override(Gravity Override=@n1)
   n2 = GetDeltaTransformFromRefPose!(self=GetOwnerSkeletalMeshComponent(), BoneName=spine_05, BaseName=spine_01) -> (ReturnValue_Location:Vector, ReturnValue_Rotation:Rotator, ReturnValue_Scale:Vector)
   > Set transform_offset(transform_offset:double=MapRangeClamped(Value=BreakRotator(InRot=@n2.ReturnValue_Rotation).Yaw, InRangeA=-15, InRangeB=-65)) -> (Output_Get:double)
  ]}
]
variables [
 Gravity Override {type=Vector}
 transform_offset {type=number}
]
